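; BP_BattleStage_Center is an environment Blueprint that animates rotating decoration meshes at the center of the S5 battle stage.
; On BeginPlay it plays a rotation timeline whose output sets the relative rotation of deco meshes C_A and C_B on a single axis.
; C_B receives the negated value, so the two pieces spin in opposite directions.
; The overlap, tick and construction script entries are empty.
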(event EventBeginPlay
  (bind (_direction _rotator) (Timeline|Rotation%20Timeline Play)
    (:Update
      (Transformation|SetRelativeRotation (Variables|Default|GetSM_BattleStage_Deco_C_A) (Math|Rotator|MakeRotator 0.0 0.0 _rotator) false false)
      (Transformation|SetRelativeRotation (Variables|Default|GetSM_BattleStage_Deco_C_B) (Math|Rotator|MakeRotator 0.0 0.0 (* _rotator -1.0)) false false))))

(event Collision|EventActorBeginOverlap
  (params (OtherActor :type Actor%20Object%20Reference)))

(event EventTick
  (params (DeltaSeconds :type Float%20%28single-precision%29)))

(fn ConstructionScript)
